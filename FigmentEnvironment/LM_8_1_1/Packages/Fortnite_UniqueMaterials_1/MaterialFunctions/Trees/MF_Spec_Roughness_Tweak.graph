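# Material function in the Trees folder that adjusts the Specular and Roughness of incoming material attributes.
# Specular is scaled by a multiplier, with back faces dimmed through TwoSidedSign when the Invert Back Facing Normals switch is off.
# An optional switch multiplies Specular by vertex color red as AO, remapped between min and max parameters.
# Roughness is then remapped by a lerp between RoughnessMin and RoughnessMax.
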MaterialFunction /FigmentEnvironment/LM_8_1_1/Packages/Fortnite_UniqueMaterials_1/MaterialFunctions/Trees/MF_Spec_Roughness_Tweak
graph {nodes [
  n1 = GetMaterialAttributes(None=FunctionInput() {inputType=FunctionInput_MaterialAttributes}) {attributeGetTypes [#1]}
  n2 = ScalarParameter() {defaultValue=0.5, parameterName=VertAOMin}
  n3 = ScalarParameter() {defaultValue=1, parameterName=VertAOMax}
  n4 = VertexColor()
  n5 = Multiply(A=@n1.Specular, B=ScalarParameter() {defaultValue=1, group=Specular, parameterName=Specular Multiplier})
  n6 = Add(A=Clamp(None=TwoSidedSign())) {constB=0.5}
  n7 = StaticSwitchParameter(True=@n5, False=Multiply(A=@n5, B=Saturate(None=@n6))) {defaultValue=true, parameterName=Invert Back Facing Normals}
  n8 = Multiply(A=ComponentMask(None=LinearInterpolate(A=@n2, B=@n3, Alpha=@n4.R)) {r=true}, B=@n7)
  n9 = GetMaterialAttributes(None=SetMaterialAttributes!(MaterialAttributes=@n1.MaterialAttributes, Specular=StaticSwitchParameter(True=@n8, False=@n7) {parameterName=UseVertexColorAsAO}) {attributeSetTypes [#1]}) {attributeGetTypes [#2]}
  n10 = ScalarParameter() {group=Specular and Roughness, parameterName=RoughnessMin}
  n11 = ScalarParameter() {defaultValue=1, group=Specular and Roughness, parameterName=RoughnessMax}
  FunctionOutput(None=SetMaterialAttributes!(MaterialAttributes=@n9.MaterialAttributes, Roughness=LinearInterpolate(A=@n10, B=@n11, Alpha=@n9.Roughness)) {attributeSetTypes [#2]})
 ]}
