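# Material function for trees that adds animated, wind-panned noise to the normal of incoming Material Attributes.
# Two noise samples at explicit mip levels scroll via vertex-interpolated time offsets using code-referenced speeds and offsets, with UV1 shifting the tiling.
# Their sum perturbs normal XY, scaled by intensity and mask, behind a static switch plus quality and feature-level fallbacks.
# The result is transformed, optionally flipped on back faces, then replaces the original normal.
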MaterialFunction /FigmentEnvironment/LM_8_1_1/Packages/Fortnite_UniqueMaterials_1/MaterialFunctions/Trees/MF_TreePanningNormals
graph {nodes [
  n1 = FunctionInput() {inputType=FunctionInput_MaterialAttributes}
  n2 = MaterialFunctionCall!(Material=@n1) {materialFunction=/Engine/Functions/MaterialLayerFunctions/MatLayerBlend_BreakNormal}
  n3 = FunctionInput() {bUsePreviewValueAsDefault=true, inputName=Mask, inputType=FunctionInput_Scalar, sortPriority=43, previewValue=<1 0 0 1>}
  n4 = TextureCoordinate() {uTiling=2, vTiling=2}
  n5 = Add(
   A=Multiply(A=@n4, B=ScalarParameter() {defaultValue=0.5, group=PX_PanningNormals, parameterName=Tiling Normal Scale})
   B=TextureCoordinate() {coordinateIndex=1, uTiling=102.5125, vTiling=101.1456}
  )
  n6 = Add(
   A=Multiply(
    A=Time() {bIgnorePause=true}
    B=ScalarParameter() {defaultValue=0.1, group=PX_PanningNormals, parameterName=Wind_Panner_Speed_1x_CodeReferenced}
   )
   B=ScalarParameter() {group=Code Controlled, parameterName=Wind_Panner_Offset_1x_CodeReferenced}
  )
  n7 = Add(
   A=Multiply(
    A=Time() {bIgnorePause=true}
    B=ScalarParameter() {defaultValue=0.1, group=PX_PanningNormals, parameterName=Wind_Panner_Speed_1y_CodeReferenced}
   )
   B=ScalarParameter() {group=Code Controlled, parameterName=Wind_Panner_Offset_1y_CodeReferenced}
  )
  n8 = Add(A=Multiply(A=@n5) {constB=1.1}, B=VertexInterpolator(VS=AppendVector(A=@n6, B=@n7)).PS)
  n9 = MaterialFunctionCall!(UVs=@n4, Texture Size=ScalarParameter() {defaultValue=128, group=PX_PanningNormals, parameterName=Panning Normal Resolution}) {materialFunction=/Engine/Functions/Engine_MaterialFunctions01/Texturing/ComputeMipLevel}
  n10 = TextureSampleParameter2D!(UVs=Frac(None=Add(A=@n8) {constB=0.1}), Level=@n9.MipLevel, Apply View MipBias) {
   group=PX_PanningNormals
   mipValueMode=TMVM_MipLevel
   parameterName=PanningNormals_1
   samplerType=SAMPLERTYPE_LinearColor
   texture=/FigmentEnvironment/LM_8_1_1/Packages/User_Built_Props/Textures/T_Normal_Noise_Uncompressed
  }
  n11 = Multiply(
   A=Time() {bIgnorePause=true}
   B=ScalarParameter() {defaultValue=0.1, group=PX_PanningNormals, parameterName=Wind_Panner_Speed_2x_CodeReferenced}
  )
  n12 = ScalarParameter() {group=Code Controlled, parameterName=Wind_Panner_Offset_2x_CodeReferenced}
  n13 = Multiply(
   A=Time() {bIgnorePause=true}
   B=ScalarParameter() {defaultValue=0.1, group=PX_PanningNormals, parameterName=Wind_Panner_Speed_2y_CodeReferenced}
  )
  n14 = ScalarParameter() {group=Code Controlled, parameterName=Wind_Panner_Offset_2y_CodeReferenced}
  n15 = VertexInterpolator(VS=AppendVector(A=Add(A=@n11, B=@n12), B=Add(A=@n13, B=@n14)))
  n16 = TextureSampleParameter2D(UVs=Frac(None=Add(A=@n5, B=@n15.PS)), Level=@n9.MipLevel) {
   group=PX_PanningNormals
   mipValueMode=TMVM_MipLevel
   parameterName=PanningNormals
   samplerType=SAMPLERTYPE_LinearColor
   texture=&T_Normal_Noise_Uncompressed
  }
  n17 = Add(A=@n10.RGB, B=@n16.RGB)
  n18 = Multiply(
   A=Multiply(A=Subtract(A=@n17), B=ScalarParameter() {defaultValue=-0.0022, group=PX_PanningNormals, parameterName=CodeRef_PanningNormalIntensity})
   B=Constant3Vector() {constant=<rgba 1 1 0 0>}
  )
  n19 = Normalize(VectorInput=Add(A=@n2.Normal, B=Multiply(A=@n3, B=@n18)))
  n20 = QualitySwitch(Default=@n19, Low=@n2.Normal, High=@n19, Medium=@n2.Normal)
  n21 = Transform(None=StaticSwitchParameter(True=FeatureLevelSwitch(Default=@n20, ES2=@n2.Normal, ES3_1=@n2.Normal, SM4_REMOVED=@n20), False=@n2.Normal) {group=PX_PanningNormals, parameterName=Use Panning Normals})
  n22 = Multiply(A=@n21, B=TwoSidedSign())
  FunctionOutput(None=MaterialFunctionCall!(Material=@n1, Normal=StaticSwitchParameter(True=@n22, False=@n21) {defaultValue=true, parameterName=Invert Back Facing Normals}) {materialFunction=/Engine/Functions/MaterialLayerFunctions/MatLayerBlend_ReplaceNormals}.Result)
 ]}
