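# Tree material function meant to fake sunlight on leaf undersides through emissive, though its output passes the input attributes straight through.
# Back faces get subsurface color times opacity, sun color and a biased sun-to-normal dot product.
# Top and bottom boost parameters blend by bounding-box height and are weighted by active time of day.
# The static switch that blends this into emissive is not wired to the output.
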MaterialFunction /FigmentEnvironment/LM_8_1_1/Packages/Fortnite_UniqueMaterials_1/MaterialFunctions/Trees/MF_Tree_FakeUnderlight
graph {nodes [
  n1 = FunctionInput() {inputType=FunctionInput_MaterialAttributes}
  FunctionOutput(None=@n1)
  n2 = BreakMaterialAttributes!(Attr=@n1)
  n3 = CollectionParameter() {
   collection=/FigmentEnvironment/LM_8_1_1/Packages/Fortress_SharedMaterials/GlobalMaterialParameters/FortniteMaterialParameters
   parameterName=SunLightColor
  }
  n4 = CollectionParameter() {
   collection=&FortniteMaterialParameters
   parameterName=SunlightDirectionalVector
  }
  n5 = ComponentMask(None=@n4) {b=true, g=true, r=true}
  n6 = VertexNormalWS()
  n7 = Multiply(A=ComponentMask(None=@n3) {b=true, g=true, r=true}, B=ConstantBiasScale(None=DotProduct(A=@n5, B=@n6)))
  n8 = OneMinus(None=Clamp(None=TwoSidedSign()))
  n9 = Multiply(A=@n2.SubsurfaceColor, B=Multiply(A=Multiply(A=@n7, B=@n8), B=@n2.Opacity))
  n10 = CollectionParameter() {
   collection=&FortniteMaterialParameters
   parameterName=FortniteActiveTimeOfDay
  }
  n11 = VectorParameter() {group=PX, parameterName=Emissive Boost To The Underside Of The Leaves Per Time of Day - Top, defaultValue=<rgba 0.1 0.1 0.05 0.1>}
  n12 = AppendVector(A=@n11.RGB, B=@n11.RGB)
  n13 = VectorParameter() {group=PX, parameterName=Emissive Boost To The Underside Of The Leaves Per Time of Day - Bottom}
  n14 = AppendVector(A=@n13.RGB, B=@n13.RGB)
  n15 = MaterialFunctionCall() {materialFunction=&BoundingBoxBased_0-1_UVW}
  n16 = DotProduct(A=Multiply(A=@n10, B=LinearInterpolate(A=@n12, B=@n14, Alpha=@n15.B)), B=Constant() {r=1})
  StaticSwitchParameter(
   True=MaterialFunctionCall!(Material=@n1, Emissive=Multiply(A=@n9, B=@n16)) {materialFunction=/Engine/Functions/MaterialLayerFunctions/MatLayerBlend_Emissive}.Blended Material
   False=@n1
  ) {group=PX, parameterName=Light Canopy Underside Via Emissive}
 ]}
