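# Material function for trees that optionally tints Base Color with snow based on the incoming material attributes.
# Use_BRSnow blends toward light grey using a biased, contrast-adjusted dot product of Normal and Z-up.
# Otherwise, UseSnow blends toward a snow color using a mask from Normal Z times base-color red, or passes attributes through when off.
# A Holiday SnowFall function call is present but does not feed the output.
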MaterialFunction /FigmentEnvironment/LM_8_1_1/Packages/Fortnite_UniqueMaterials_1/MaterialFunctions/Trees/MF_Tree_Snow
graph {nodes [
  Constant3Vector()
  Constant() {r=1}
  Constant()
  n1 = GetMaterialAttributes(None=FunctionInput() {inputType=FunctionInput_MaterialAttributes}) {attributeGetTypes [#1, #2]}
  n2 = MaterialFunctionCall(In=DotProduct(A=Normalize(VectorInput=@n1.Normal), B=Constant3Vector() {constant=<rgba 0 0 1 0>}), Contrast=Constant()) {materialFunction=&CheapContrast}
  n3 = ScalarParameter() {defaultValue=-0.2, parameterName=SnowNormal_Bias}
  n4 = LinearInterpolate(A=@n1.BaseColor, Alpha=Saturate(None=Add(A=@n2.Result, B=@n3))) {constB=0.7}
  n5 = Clamp(None=LinearInterpolate(
    A=ScalarParameter() {defaultValue=-0.0381, group=Snow, parameterName=SnowNormalContrast1}
    B=ScalarParameter() {defaultValue=1.4108, group=Snow, parameterName=SnowNormalContrast2}
    Alpha=ComponentMask(None=@n1.Normal) {b=true}
   ))
  n6 = Clamp(None=LinearInterpolate(
    A=ScalarParameter() {group=Snow, parameterName=SnowDiffuseContrast1}
    B=ScalarParameter() {defaultValue=13.0863, group=Snow, parameterName=SnowDiffuseContrast2}
    Alpha=ComponentMask(None=@n1.BaseColor) {r=true}
   ))
  n7 = SetMaterialAttributes!(
   MaterialAttributes=@n1.MaterialAttributes
   BaseColor=LinearInterpolate(
    A=@n1.BaseColor
    B=VectorParameter() {group=Snow, parameterName=SnowColor_1, defaultValue=<rgba 0.7262 0.7933 0.83 1>}.RGB
    Alpha=Multiply(A=@n5, B=@n6)
   )
  ) {attributeSetTypes [#1]}
  FunctionOutput(None=StaticSwitchParameter(
    True=SetMaterialAttributes(MaterialAttributes=@n1.MaterialAttributes, BaseColor=@n4) {attributeSetTypes [#1]}
    False=StaticSwitchParameter(True=@n7, False=@n1.MaterialAttributes) {parameterName=UseSnow}
   ) {parameterName=Use_BRSnow})
  MaterialFunctionCall!(
   Use_ExtraSnowMask
   UV_WorldMask
   SnowStartBiasOverride
   UseHeightLerp
   LandscapeBlend_Height
   Landscape_RockMask
   Snow_Mask
   Simple_UseWorldZNormal=StaticBool() {value=true}
   Simple_WorldNormal=@n1.Normal
   WorldPosition=StaticSwitch(True=ActorPositionWS(), False=WorldPosition().XYZ, Value=StaticBoolParameter() {defaultValue=true, parameterName=UseActorPositionForSnow})
  ) {materialFunction=/FigmentEnvironment/LM_8_1_1/Athena/Environments/Landscape/MaterialFunctions/Snow/MF_Athena_Holiday_SnowFall}
  ObjectPositionWS()
 ]}
